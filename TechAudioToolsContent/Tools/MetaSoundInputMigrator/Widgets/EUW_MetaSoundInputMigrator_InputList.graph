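# Editor utility list panel for the MetaSound Input Migrator tool, with configurable header, asset picker, search box, entry list and instructions.
# A SinglePropertyView edits the widget's own Metasound variable, broadcasting a FieldNotify change and clearing list selection when invalid.
# Hovering an entry passes hover state to its linked entry viewmodel if enabled, and a function pushes Show Default Value to all entries.
# Panel hover recolors text and header icon from the style viewmodel.
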EditorUtilityWidgetBlueprint /TechAudioToolsContent/Tools/MetaSoundInputMigrator/Widgets/EUW_MetaSoundInputMigrator_InputList
graphs [
 EventGraph {nodes [
   n1 = "On Item Is Hovered Changed (ListView)"!() -> (OutputDelegate, then, Item:Object Reference, bIsHovered:bool) {is=title K2Node_ComponentBoundEvent}
   > n2 = Cast To BP_MetaSoundInputMigratorEntryViewModel!(execute, Object=@n1.Item) -> (then, CastFailed, AsBP Meta Sound Input Migrator Entry View Model:BP Meta Sound Input Migrator Entry View Model Object Reference) {is=title K2Node_DynamicCast}
   > n3 = Get!(execute:Exec, self:BP Meta Sound Input Migrator Entry View Model Object Reference=@n2) -> (then:Exec, else:Exec, Linked Migrator Entry Viewmodel:BP Meta Sound Input Migrator Entry View Model Object Reference)
   > Branch(Condition=Get bIsEnabled!(self:Tech Audio Tools Entry Viewmodel Object Reference=@n3) -> (bIsEnabled:bool))
   > Set bIsHovered!(execute, bIsHovered:bool=@n1.bIsHovered, self:Tech Audio Tools Entry Viewmodel Object Reference=@n3) -> (then, Output_Get:bool)
   "On Property Changed (MetaSoundAssetPicker)"!() -> (OutputDelegate, then, PropertyName:Name=None) {is=title K2Node_ComponentBoundEvent}
   > Broadcast Field Value Changed!(execute, self:Widget Object Reference, FieldId:Field Notification Id Structure=(FieldName="Metasound")) -> (then)
   > n4 = Branch(Condition=IsValidInterface!(Interface:"Interface Interface (by ref)"=Get Metasound() -> (Metasound:Meta Sound Document Interface Interface)) -> (ReturnValue:bool))
   Clear Selection(execute=@n4.else, self:List View Object Reference=Get ListView() -> (ListView:List View Object Reference))
   Event PreConstruct()
   > SetObject(self=Get MetaSoundAssetPicker() -> (MetaSoundAssetPicker:Single Property View Object Reference), NewObject=Self-Reference())
   > "SetText (Text)"(self=Get Header_Text() -> (Header_Text:Common Text Object Reference), InText=Get Header Text() -> (Header Text:Text))
   > SetIconBrushByPath!(
    execute
    self:Tech Audio Tools Editor Slate Icon Object Reference=Get Header_Icon() -> (Header_Icon:Tech Audio Tools Editor Slate Icon Object Reference)
    InFullIconPath:Name=Get Header Icon Path() -> (Header Icon Path:Name=None)
   ) -> (then)
   > "SetText (Text)"(self=Get Instructions_Text() -> (Instructions_Text:Text Object Reference), InText=Get Instructions Text() -> (Instructions Text:Text))
   > SetIconBrushByPath(
    self=Get Instructions_Icon() -> (Instructions_Icon:Tech Audio Tools Editor Slate Icon Object Reference)
    InFullIconPath=Get Instructions Icon Path() -> (Instructions Icon Path:Name=None)
   )
   Event OnMouseEnter()
   > Update Header Color()
   Event OnMouseLeave!() -> (OutputDelegate, then, MouseEvent:"Pointer Event Structure (by ref)") {is=title K2Node_Event}
   > Update Header Color()
   Event Construct()
   > Update Header Color()
  ]}
 Set Show Default Values {nodes [
   Set Show Default Values!() -> (then, Show Default Value:bool=~) {is=title K2Node_FunctionEntry}
   > n1 = For Each Loop(Array:Array of Tech Audio Tools Entry Viewmodel Object References=GetEntryViewModels!(self:Meta Sound Member List Viewmodel Object Reference=Get MetaSoundInputMigratorListViewModel() -> (MetaSoundInputMigratorListViewModel:BP Meta Sound Input Migrator List View Model Object Reference)) -> (ReturnValue:Array of Tech Audio Tools Entry Viewmodel Object References)) -> (Array Element:Tech Audio Tools Entry Viewmodel Object Reference)
   > n2 = Cast To BP_MetaSoundInputMigratorEntryViewModel(Object=@n1)
   > Set with Broadcast Show Default Value!(
    execute
    Show Default Value:bool=Get Show Default Value() -> (Show Default Value:bool)
    self:BP Meta Sound Input Migrator Entry View Model Object Reference=@n2
   ) -> (then, Output_Get:bool)
  ]}
 Update Header Color {nodes [
   Update Header Color() {is=title K2Node_FunctionEntry}
   > n1 = Get!(execute) -> (then, else, TechAudioToolsListViewModel:Tech Audio Tools List Viewmodel Object Reference)
   > n2 = Get!(execute, self:Tech Audio Tools List Viewmodel Object Reference=@n1) -> (then, else, StyleViewModel:Tech Audio Tools Style Viewmodel Object Reference)
   n3 = SelectColor(
    A=Get Text2!(self:Tech Audio Tools Style Viewmodel Object Reference=@n2) -> (Text2:Linear Color Structure)
    B=Get Text1!(self:Tech Audio Tools Style Viewmodel Object Reference=@n2) -> (Text1:Linear Color Structure)
    bPickA=IsHovered()
   )
   > SetColorAndOpacity(self=[Get Instructions_Text(), Get Header_Text()], InColorAndOpacity=Make Slate Color(SpecifiedColor=@n3))
   > SetIconColor!(execute, self:Tech Audio Tools Editor Slate Icon Object Reference=Get Header_Icon(), InColor:Linear Color Structure=@n3) -> (then)
  ]}
]
variables [
 Header Icon Path {type=string, default=EditorStyle.Icons.CircleArrowRight}
 Header Text {type=string, default=MetaSound}
 Instructions Icon Path {type=string, default=CoreStyle.Icons.CircleArrowRight}
 Instructions Text {type=string, default=Instructions text.}
 Metasound {default=None}
]
widgets {
 info {
  inheritedWidgetCount=1
  parentClass=/Script/TechAudioToolsContentEditor.TechAudioToolsEditorListView
  rootWidgetClass=&VerticalBox
  widgetCount=12
 }
 widgets [
  VerticalBox_0 {widgetClassPath=&VerticalBox, widget=VerticalBox_0}
  Border_9 {
   widgetClassPath=&Border
   parent=VerticalBox_0
   widget=Border_9
   properties {
    background {
     bIsDynamicallyLoaded=false
     drawAs=RoundedBox
     imageType=NoImage
     mirroring=NoMirror
     resourceName=None
     resourceObject=null
     tiling=NoTile
     imageSize=<32 32>
     margin=<ltrb 0 0 0 0>
     outlineSettings {
      bUseBrushTransparency=false
      roundingType=FixedRadius
      width=0
      color {colorUseRule=UseColor_Specified, specifiedColor=<rgba 0.0176 0.0176 0.0176 1>}
      cornerRadii=<4 4 0 0>
     }
     tintColor {colorUseRule=UseColor_Specified, specifiedColor=<rgba 1 1 1 1>}
     uVRegion {bIsValid=false, max=<0 0>, min=<0 0>}
    }
    brushColor=<rgba 0.0176 0.0176 0.0176 1>
    padding=<ltrb 0 0 0 0>
   }
   slot {class=VerticalBoxSlot}
  }
  VerticalBox_1 {widgetClassPath=&VerticalBox, parent=Border_9, widget=VerticalBox_1, slot {class=BorderSlot, properties {padding=<ltrb 0 0 0 0>}}}
  HorizontalBox_2 {widgetClassPath=&HorizontalBox, parent=VerticalBox_1, widget=HorizontalBox_2, slot {class=VerticalBoxSlot}}
  Header_Icon {
   widgetClassPath=/Script/TechAudioToolsContentEditor.TechAudioToolsEditorSlateIcon
   bIsVariable=true
   parent=HorizontalBox_2
   widget=Header_Icon
   properties {fullIconPath=EditorStyle.Icons.CircleArrowRight, imageColorAndOpacity=<rgba 0.5271 0.5271 0.5271 1>}
   slot {class=HorizontalBoxSlot, properties {verticalAlignment=VAlign_Center, padding=<ltrb 6 6 6 6>}}
  }
  Header_Text {
   widgetClassPath=&CommonTextBlock
   bIsVariable=true
   parent=HorizontalBox_2
   widget=Header_Text
   properties {
    text=MetaSound
    colorAndOpacity {colorUseRule=UseColor_Specified, specifiedColor=<rgba 0.5271 0.5271 0.5271 1>}
    font {
     bForceMonospaced=false
     bMaterialIsStencil=false
     fontMaterial=null
     fontObject=&Roboto
     letterSpacing=0
     monospacedWidth=1
     size=9.38
     skewAmount=0
     typefaceFontName=Regular
     outlineSettings {
      bApplyOutlineToDropShadows=false
      bMiteredCorners=false
      bSeparateFillAlpha=false
      outlineMaterial=null
      outlineSize=0
      outlineColor=<rgba 0 0 0 1>
     }
    }
   }
   slot {class=HorizontalBoxSlot, properties {verticalAlignment=VAlign_Center, padding=<ltrb 0 0 24 0>}}
  }
  MetaSoundAssetPicker {
   widgetClassPath=&SinglePropertyView
   parent=VerticalBox_1
   widget=MetaSoundAssetPicker
   properties {nameOverride=" ", propertyName=Metasound}
   slot {class=VerticalBoxSlot, properties {padding=<ltrb 0 0 0 8>}}
  }
  SearchBox {
   widgetClassPath=/TechAudioToolsContent/Widgets/Editor/EUW_SearchBox
   bIsVariable=true
   parent=VerticalBox_1
   widget=SearchBox
   slot {class=VerticalBoxSlot, properties {padding=<ltrb 2 2 2 2>}}
  }
  ListView {
   widgetClassPath=/Script/Blutility.EditorUtilityListView
   bInherited=true
   parent=VerticalBox_0
   widget=ListView
   properties {
    bEnableRightClickScrolling=false
    bEnableScrollAnimation=true
    dragDropOperationClass=/Script/UMG.DragDropOperation
    dragDropVisualEntryClass=&/EUW_MetaSoundInputMigratorEntry.EUW_MetaSoundInputMigratorEntry_C
    dragDropVisualPivot=CenterLeft
    entryWidgetClass=&EUW_MetaSoundInputMigratorEntry_C
    numDesignerPreviewEntries=10
    selectionMode=SingleToggle
   }
   slot {class=VerticalBoxSlot, properties {size {sizeRule=Fill, value=1}}}
  }
  VerticalBox_113 {
   widgetClassPath=&VerticalBox
   parent=VerticalBox_0
   widget=VerticalBox_113
   slot {class=VerticalBoxSlot, properties {size {sizeRule=Fill, value=1}}}
  }
  Instructions_Icon {
   widgetClassPath=&TechAudioToolsEditorSlateIcon
   bIsVariable=true
   parent=VerticalBox_113
   widget=Instructions_Icon
   properties {imageColorAndOpacity=<rgba 0.0284 0.0284 0.0284 1>, size=<64 64>}
   slot {class=VerticalBoxSlot, properties {horizontalAlignment=HAlign_Center, verticalAlignment=VAlign_Center, padding=<ltrb 32 32 32 32>}}
  }
  Instructions_Text {
   widgetClassPath=&TextBlock
   bIsVariable=true
   parent=VerticalBox_113
   widget=Instructions_Text
   properties {
    justification=Center
    text=Instructions text.
    colorAndOpacity {colorUseRule=UseColor_Specified, specifiedColor=<rgba 0.5271 0.5271 0.5271 1>}
    font {
     bForceMonospaced=false
     bMaterialIsStencil=false
     fontMaterial=null
     fontObject=&Roboto
     letterSpacing=0
     monospacedWidth=1
     size=9
     skewAmount=0
     typefaceFontName=Italic
     outlineSettings {
      bApplyOutlineToDropShadows=false
      bMiteredCorners=false
      bSeparateFillAlpha=false
      outlineMaterial=null
      outlineSize=0
      outlineColor=<rgba 0 0 0 1>
     }
    }
   }
   slot {class=VerticalBoxSlot, properties {horizontalAlignment=HAlign_Center, verticalAlignment=VAlign_Center}}
  }
 ]
}
