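# Editor Utility split button for Tech Audio Tools: an icon-and-label main button beside a narrow ellipsis menu button.
# PreConstruct applies the configurable icon, size and label, while Update Icon Color highlights whichever half is hovered.
# Hovering either half swaps the other half's style to show its rounded backing, so both read as one control.
# Clicks fire separate primary and menu dispatchers; custom events enable each half independently, and no menu is built here.
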EditorUtilityWidgetBlueprint /TechAudioToolsContent/Widgets/Editor/Buttons/EUW_Button_WithMenu
graphs [
 EventGraph {nodes [
   "On Hovered (Menu_Button)"() {is=title K2Node_ComponentBoundEvent}
   > SetStyle!(
    execute
    self:Button Object Reference=Get Button() -> (Button:Editor Utility Button Object Reference)
    InStyle:"Button Style Structure (by ref)"=Get Main Button Style Options Hovered() -> (Main Button Style Options Hovered:Button Style Structure)
   ) -> (then)
   > Update Icon Color()
   "On Clicked (Button)"() {is=title K2Node_ComponentBoundEvent}
   > Call On Primary Button Clicked!(execute:Exec, self:Tech Audio Tools Button With Menu Object Reference) -> (then:Exec) {is=title K2Node_CallDelegate}
   "On Unhovered (Menu_Button)"() {is=title K2Node_ComponentBoundEvent}
   > SetStyle(self=Get Button(), InStyle=Get Main Button Style() -> (Main Button Style:Button Style Structure))
   > Update Icon Color()
   "On Hovered (Button)"() {is=title K2Node_ComponentBoundEvent}
   > SetStyle(
    self=Get Menu_Button() -> (Menu_Button:Editor Utility Button Object Reference)
    InStyle=Get Options Button Style Main Hovered() -> (Options Button Style Main Hovered:Button Style Structure)
   )
   > Update Icon Color()
   "On Unhovered (Button)"() {is=title K2Node_ComponentBoundEvent}
   > SetStyle(self=Get Menu_Button(), InStyle=Get Options Button Style() -> (Options Button Style:Button Style Structure))
   > Update Icon Color()
   "On Clicked (Menu_Button)"() {is=title K2Node_ComponentBoundEvent}
   > Call On Menu Button Clicked!(execute:Exec, self:Tech Audio Tools Button With Menu Object Reference) -> (then:Exec) {is=title K2Node_CallDelegate}
   Update Icon Color() {is=title K2Node_CustomEvent}
   n1 = Get Button()
   > SetIconColor!(
    execute
    self:Tech Audio Tools Editor Slate Icon Object Reference=Get Icon() -> (Icon:Tech Audio Tools Editor Slate Icon Object Reference)
    InColor:Linear Color Structure=SelectColor(
     A=Get Highlighted Icon Color() -> (Highlighted Icon Color:Linear Color Structure)
     B=Get Default Icon Color() -> (Default Icon Color:Linear Color Structure)
     bPickA=IsHovered(self=@n1)
    )
   ) -> (then)
   n2 = IsHovered(self=Get Button())
   n3 = Get Default Icon Color()
   n4 = Get Highlighted Icon Color()
   > SetColorAndOpacity(
    self=Get Label_Text() -> (Label_Text:Text Object Reference)
    InColorAndOpacity=Make Slate Color(SpecifiedColor=SelectColor(A=@n4, B=@n3, bPickA=@n2))
   )
   n5 = Get Menu_Button()
   > SetIconColor(
    self=Get Menu_Icon() -> (Menu_Icon:Tech Audio Tools Editor Slate Icon Object Reference)
    InColor=SelectColor(A=Get Highlighted Icon Color(), B=Get Default Icon Color(), bPickA=IsHovered(self=@n5))
   )
   n6 = Set Main Button Enabled!() -> (OutputDelegate, then, Is Enabled:bool=~) {is=title K2Node_CustomEvent}
   > SetIsEnabled(self=Get Button(), bInIsEnabled=@n6.Is Enabled)
   n7 = Set Menu Button Enabled!() -> (OutputDelegate, then, Is Enabled:bool=~) {is=title K2Node_CustomEvent}
   > SetIsEnabled(self=Get Menu_Button(), bInIsEnabled=@n7.Is Enabled)
   Event PreConstruct()
   > SetIconBrushByPath!(
    execute
    self:Tech Audio Tools Editor Slate Icon Object Reference=Get Icon()
    InFullIconPath:Name=Get Icon Brush Path() -> (Icon Brush Path:Name=None)
   ) -> (then)
   > SetIconSize!(
    execute
    self:Tech Audio Tools Editor Slate Icon Object Reference=Get Icon()
    InSize:Vector 2D Structure=Get Icon Size() -> (Icon Size:Vector 2D Structure)
   ) -> (then)
   > Update Icon Color() {is=title K2Node_CallFunction}
   > "SetText (Text)"(self=Get Label_Text(), InText=Get Label Text() -> (Label Text:Text))
   Event OnMouseEnter()
   > Update Icon Color()
   Event OnMouseLeave!() -> (OutputDelegate, then, MouseEvent:"Pointer Event Structure (by ref)") {is=title K2Node_Event}
   > Update Icon Color()
  ]}
 On Menu Button Clicked {nodes [On Menu Button Clicked() {is=title K2Node_FunctionEntry}]}
 On Primary Button Clicked {nodes [On Primary Button Clicked() {is=title K2Node_FunctionEntry}]}
]
variables [
 Default Icon Color {type=LinearColor, default=<rgba 0.5271 0.5271 0.5271 1>}
 Highlighted Icon Color {type=LinearColor, default=<rgba 1 1 1 1>}
 Icon Brush Path {type=string, default=EditorStyle.Icons.Save}
 Icon Size {type=Vector2D, default=<20 20>}
 Label Text {type=string, default=Button}
 Main Button Style {
  type=ButtonStyle
  default {
   clickedSlateSound {resourceObject=null}
   disabled {
    bIsDynamicallyLoaded=false
    drawAs=NoDrawType
    imageType=NoImage
    mirroring=NoMirror
    resourceName=None
    resourceObject=null
    tiling=NoTile
    imageSize=<32 32>
    margin=<ltrb 0 0 0 0>
    outlineSettings {
     bUseBrushTransparency=false
     roundingType=FixedRadius
     width=1
     color {colorUseRule=UseColor_ColorTable, specifiedColor=<rgba 0 0 0 0>}
     cornerRadii=<4 4 4 4>
    }
    tintColor {colorUseRule=UseColor_Specified, specifiedColor=<rgba 0.0312 0.0312 0.0312 1>}
    uVRegion {bIsValid=false, max=<0 0>, min=<0 0>}
   }
   disabledForeground {colorUseRule=UseColor_ColorTable, specifiedColor=<rgba 0 0 0 0>}
   hovered {
    bIsDynamicallyLoaded=false
    drawAs=RoundedBox
    imageType=NoImage
    mirroring=NoMirror
    resourceName=None
    resourceObject=null
    tiling=NoTile
    imageSize=<32 32>
    margin=<ltrb 0 0 0 0>
    outlineSettings {
     bUseBrushTransparency=false
     roundingType=FixedRadius
     width=0
     color {colorUseRule=UseColor_Foreground, specifiedColor=<rgba 0 0 0 0>}
     cornerRadii=<4 0 0 4>
    }
    tintColor {colorUseRule=UseColor_Specified, specifiedColor=<rgba 0.0953 0.0953 0.0953 1>}
    uVRegion {bIsValid=false, max=<0 0>, min=<0 0>}
   }
   hoveredForeground {colorUseRule=UseColor_ColorTable, specifiedColor=<rgba 0 0 0 0>}
   hoveredSlateSound {resourceObject=null}
   normal {
    bIsDynamicallyLoaded=false
    drawAs=NoDrawType
    imageType=NoImage
    mirroring=NoMirror
    resourceName=None
    resourceObject=null
    tiling=NoTile
    imageSize=<32 32>
    margin=<ltrb 0 0 0 0>
    outlineSettings {
     bUseBrushTransparency=false
     roundingType=FixedRadius
     width=1
     color {colorUseRule=UseColor_ColorTable, specifiedColor=<rgba 0 0 0 0>}
     cornerRadii=<4 4 4 4>
    }
    tintColor {colorUseRule=UseColor_Specified, specifiedColor=<rgba 0.0312 0.0312 0.0312 1>}
    uVRegion {bIsValid=false, max=<0 0>, min=<0 0>}
   }
   normalForeground {colorUseRule=UseColor_ColorTable, specifiedColor=<rgba 0 0 0 0>}
   normalPadding=<ltrb 2 0 2 0>
   pressed {
    bIsDynamicallyLoaded=false
    drawAs=RoundedBox
    imageType=NoImage
    mirroring=NoMirror
    resourceName=None
    resourceObject=null
    tiling=NoTile
    imageSize=<32 32>
    margin=<ltrb 0 0 0 0>
    outlineSettings {
     bUseBrushTransparency=false
     roundingType=FixedRadius
     width=0
     color {colorUseRule=UseColor_Foreground, specifiedColor=<rgba 0 0 0 0>}
     cornerRadii=<4 0 0 4>
    }
    tintColor {colorUseRule=UseColor_Specified, specifiedColor=<rgba 0.0395 0.0395 0.0395 1>}
    uVRegion {bIsValid=false, max=<0 0>, min=<0 0>}
   }
   pressedForeground {colorUseRule=UseColor_ColorTable, specifiedColor=<rgba 0 0 0 0>}
   pressedPadding=<ltrb 2 0.5 2 -0.5>
   pressedSlateSound {resourceObject=null}
  }
 }
 Main Button Style Options Hovered {
  type=ButtonStyle
  default {
   clickedSlateSound {resourceObject=null}
   disabled {
    bIsDynamicallyLoaded=false
    drawAs=NoDrawType
    imageType=NoImage
    mirroring=NoMirror
    resourceName=None
    resourceObject=null
    tiling=NoTile
    imageSize=<32 32>
    margin=<ltrb 0 0 0 0>
    outlineSettings {
     bUseBrushTransparency=false
     roundingType=FixedRadius
     width=1
     color {colorUseRule=UseColor_ColorTable, specifiedColor=<rgba 0 0 0 0>}
     cornerRadii=<4 4 4 4>
    }
    tintColor {colorUseRule=UseColor_Specified, specifiedColor=<rgba 0.0312 0.0312 0.0312 1>}
    uVRegion {bIsValid=false, max=<0 0>, min=<0 0>}
   }
   disabledForeground {colorUseRule=UseColor_ColorTable, specifiedColor=<rgba 0 0 0 0>}
   hovered {
    bIsDynamicallyLoaded=false
    drawAs=RoundedBox
    imageType=NoImage
    mirroring=NoMirror
    resourceName=None
    resourceObject=null
    tiling=NoTile
    imageSize=<32 32>
    margin=<ltrb 0 0 0 0>
    outlineSettings {
     bUseBrushTransparency=false
     roundingType=FixedRadius
     width=0
     color {colorUseRule=UseColor_Foreground, specifiedColor=<rgba 0 0 0 0>}
     cornerRadii=<4 0 0 4>
    }
    tintColor {colorUseRule=UseColor_Specified, specifiedColor=<rgba 0.0953 0.0953 0.0953 1>}
    uVRegion {bIsValid=false, max=<0 0>, min=<0 0>}
   }
   hoveredForeground {colorUseRule=UseColor_ColorTable, specifiedColor=<rgba 0 0 0 0>}
   hoveredSlateSound {resourceObject=null}
   normal {
    bIsDynamicallyLoaded=false
    drawAs=RoundedBox
    imageType=NoImage
    mirroring=NoMirror
    resourceName=None
    resourceObject=null
    tiling=NoTile
    imageSize=<32 32>
    margin=<ltrb 0 0 0 0>
    outlineSettings {
     bUseBrushTransparency=false
     roundingType=FixedRadius
     width=1
     color {colorUseRule=UseColor_Specified, specifiedColor=<rgba 0.0262 0.0262 0.0262 1>}
     cornerRadii=<4 0 0 4>
    }
    tintColor {colorUseRule=UseColor_Specified, specifiedColor=<rgba 0.0395 0.0395 0.0395 1>}
    uVRegion {bIsValid=false, max=<0 0>, min=<0 0>}
   }
   normalForeground {colorUseRule=UseColor_ColorTable, specifiedColor=<rgba 0 0 0 0>}
   normalPadding=<ltrb 2 0 2 0>
   pressed {
    bIsDynamicallyLoaded=false
    drawAs=RoundedBox
    imageType=NoImage
    mirroring=NoMirror
    resourceName=None
    resourceObject=null
    tiling=NoTile
    imageSize=<32 32>
    margin=<ltrb 0 0 0 0>
    outlineSettings {
     bUseBrushTransparency=false
     roundingType=FixedRadius
     width=0
     color {colorUseRule=UseColor_Foreground, specifiedColor=<rgba 0 0 0 0>}
     cornerRadii=<4 0 0 4>
    }
    tintColor {colorUseRule=UseColor_Specified, specifiedColor=<rgba 0.0395 0.0395 0.0395 1>}
    uVRegion {bIsValid=false, max=<0 0>, min=<0 0>}
   }
   pressedForeground {colorUseRule=UseColor_ColorTable, specifiedColor=<rgba 0 0 0 0>}
   pressedPadding=<ltrb 2 0.5 2 -0.5>
   pressedSlateSound {resourceObject=null}
  }
 }
 On Menu Button Clicked {default=()}
 On Primary Button Clicked {default=()}
 Options Button Style {
  type=ButtonStyle
  default {
   clickedSlateSound {resourceObject=null}
   disabled {
    bIsDynamicallyLoaded=false
    drawAs=NoDrawType
    imageType=NoImage
    mirroring=NoMirror
    resourceName=None
    resourceObject=null
    tiling=NoTile
    imageSize=<32 32>
    margin=<ltrb 0 0 0 0>
    outlineSettings {
     bUseBrushTransparency=false
     roundingType=FixedRadius
     width=1
     color {colorUseRule=UseColor_ColorTable, specifiedColor=<rgba 0 0 0 0>}
     cornerRadii=<4 4 4 4>
    }
    tintColor {colorUseRule=UseColor_Specified, specifiedColor=<rgba 0.0312 0.0312 0.0312 1>}
    uVRegion {bIsValid=false, max=<0 0>, min=<0 0>}
   }
   disabledForeground {colorUseRule=UseColor_ColorTable, specifiedColor=<rgba 0 0 0 0>}
   hovered {
    bIsDynamicallyLoaded=false
    drawAs=RoundedBox
    imageType=NoImage
    mirroring=NoMirror
    resourceName=None
    resourceObject=null
    tiling=NoTile
    imageSize=<32 32>
    margin=<ltrb 0 0 0 0>
    outlineSettings {
     bUseBrushTransparency=false
     roundingType=FixedRadius
     width=0
     color {colorUseRule=UseColor_Foreground, specifiedColor=<rgba 0 0 0 0>}
     cornerRadii=<0 4 4 0>
    }
    tintColor {colorUseRule=UseColor_Specified, specifiedColor=<rgba 0.0953 0.0953 0.0953 1>}
    uVRegion {bIsValid=false, max=<0 0>, min=<0 0>}
   }
   hoveredForeground {colorUseRule=UseColor_ColorTable, specifiedColor=<rgba 0 0 0 0>}
   hoveredSlateSound {resourceObject=null}
   normal {
    bIsDynamicallyLoaded=false
    drawAs=NoDrawType
    imageType=NoImage
    mirroring=NoMirror
    resourceName=None
    resourceObject=null
    tiling=NoTile
    imageSize=<32 32>
    margin=<ltrb 0 0 0 0>
    outlineSettings {
     bUseBrushTransparency=false
     roundingType=FixedRadius
     width=1
     color {colorUseRule=UseColor_ColorTable, specifiedColor=<rgba 0 0 0 0>}
     cornerRadii=<4 4 4 4>
    }
    tintColor {colorUseRule=UseColor_Specified, specifiedColor=<rgba 0.0312 0.0312 0.0312 1>}
    uVRegion {bIsValid=false, max=<0 0>, min=<0 0>}
   }
   normalForeground {colorUseRule=UseColor_ColorTable, specifiedColor=<rgba 0 0 0 0>}
   normalPadding=<ltrb 2 0 2 0>
   pressed {
    bIsDynamicallyLoaded=false
    drawAs=RoundedBox
    imageType=NoImage
    mirroring=NoMirror
    resourceName=None
    resourceObject=null
    tiling=NoTile
    imageSize=<32 32>
    margin=<ltrb 0 0 0 0>
    outlineSettings {
     bUseBrushTransparency=false
     roundingType=FixedRadius
     width=0
     color {colorUseRule=UseColor_Foreground, specifiedColor=<rgba 0 0 0 0>}
     cornerRadii=<0 4 4 0>
    }
    tintColor {colorUseRule=UseColor_Specified, specifiedColor=<rgba 0.0395 0.0395 0.0395 1>}
    uVRegion {bIsValid=false, max=<0 0>, min=<0 0>}
   }
   pressedForeground {colorUseRule=UseColor_ColorTable, specifiedColor=<rgba 0 0 0 0>}
   pressedPadding=<ltrb 2 0.5 2 -0.5>
   pressedSlateSound {resourceObject=null}
  }
 }
 Options Button Style Main Hovered {
  type=ButtonStyle
  default {
   clickedSlateSound {resourceObject=null}
   disabled {
    bIsDynamicallyLoaded=false
    drawAs=NoDrawType
    imageType=NoImage
    mirroring=NoMirror
    resourceName=None
    resourceObject=null
    tiling=NoTile
    imageSize=<32 32>
    margin=<ltrb 0 0 0 0>
    outlineSettings {
     bUseBrushTransparency=false
     roundingType=FixedRadius
     width=1
     color {colorUseRule=UseColor_ColorTable, specifiedColor=<rgba 0 0 0 0>}
     cornerRadii=<4 4 4 4>
    }
    tintColor {colorUseRule=UseColor_Specified, specifiedColor=<rgba 0.0312 0.0312 0.0312 1>}
    uVRegion {bIsValid=false, max=<0 0>, min=<0 0>}
   }
   disabledForeground {colorUseRule=UseColor_ColorTable, specifiedColor=<rgba 0 0 0 0>}
   hovered {
    bIsDynamicallyLoaded=false
    drawAs=RoundedBox
    imageType=NoImage
    mirroring=NoMirror
    resourceName=None
    resourceObject=null
    tiling=NoTile
    imageSize=<32 32>
    margin=<ltrb 0 0 0 0>
    outlineSettings {
     bUseBrushTransparency=false
     roundingType=FixedRadius
     width=0
     color {colorUseRule=UseColor_Foreground, specifiedColor=<rgba 0 0 0 0>}
     cornerRadii=<0 4 4 0>
    }
    tintColor {colorUseRule=UseColor_Specified, specifiedColor=<rgba 0.0953 0.0953 0.0953 1>}
    uVRegion {bIsValid=false, max=<0 0>, min=<0 0>}
   }
   hoveredForeground {colorUseRule=UseColor_ColorTable, specifiedColor=<rgba 0 0 0 0>}
   hoveredSlateSound {resourceObject=null}
   normal {
    bIsDynamicallyLoaded=false
    drawAs=RoundedBox
    imageType=NoImage
    mirroring=NoMirror
    resourceName=None
    resourceObject=null
    tiling=NoTile
    imageSize=<32 32>
    margin=<ltrb 0 0 0 0>
    outlineSettings {
     bUseBrushTransparency=false
     roundingType=FixedRadius
     width=1
     color {colorUseRule=UseColor_Specified, specifiedColor=<rgba 0.0262 0.0262 0.0262 1>}
     cornerRadii=<0 4 4 0>
    }
    tintColor {colorUseRule=UseColor_Specified, specifiedColor=<rgba 0.0395 0.0395 0.0395 1>}
    uVRegion {bIsValid=false, max=<0 0>, min=<0 0>}
   }
   normalForeground {colorUseRule=UseColor_ColorTable, specifiedColor=<rgba 0 0 0 0>}
   normalPadding=<ltrb 2 0 2 0>
   pressed {
    bIsDynamicallyLoaded=false
    drawAs=RoundedBox
    imageType=NoImage
    mirroring=NoMirror
    resourceName=None
    resourceObject=null
    tiling=NoTile
    imageSize=<32 32>
    margin=<ltrb 0 0 0 0>
    outlineSettings {
     bUseBrushTransparency=false
     roundingType=FixedRadius
     width=0
     color {colorUseRule=UseColor_Foreground, specifiedColor=<rgba 0 0 0 0>}
     cornerRadii=<0 4 4 0>
    }
    tintColor {colorUseRule=UseColor_Specified, specifiedColor=<rgba 0.0395 0.0395 0.0395 1>}
    uVRegion {bIsValid=false, max=<0 0>, min=<0 0>}
   }
   pressedForeground {colorUseRule=UseColor_ColorTable, specifiedColor=<rgba 0 0 0 0>}
   pressedPadding=<ltrb 2 0.5 2 -0.5>
   pressedSlateSound {resourceObject=null}
  }
 }
]
widgets {
 info {rootWidgetClass=&VerticalBox, widgetCount=10}
 widgets [
  VerticalBox_0 {widgetClassPath=&VerticalBox, widget=VerticalBox_0}
  HorizontalBox_0 {widgetClassPath=&HorizontalBox, parent=VerticalBox_0, widget=HorizontalBox_0, slot {class=VerticalBoxSlot}}
  Border_0 {
   widgetClassPath=&Border
   parent=HorizontalBox_0
   widget=Border_0
   properties {brushColor=<rgba 0.0176 0.0176 0.0176 0>, padding=<ltrb 0 0 0 0>}
   slot {class=HorizontalBoxSlot}
  }
  Button {
   widgetClassPath=&EditorUtilityButton
   parent=Border_0
   widget=Button
   properties {widgetStyle {
     clickedSlateSound {resourceObject=null}
     disabled {
      bIsDynamicallyLoaded=false
      drawAs=NoDrawType
      imageType=NoImage
      mirroring=NoMirror
      resourceName=None
      resourceObject=null
      tiling=NoTile
      imageSize=<32 32>
      margin=<ltrb 0 0 0 0>
      outlineSettings {
       bUseBrushTransparency=false
       roundingType=FixedRadius
       width=1
       color {colorUseRule=UseColor_ColorTable, specifiedColor=<rgba 0 0 0 0>}
       cornerRadii=<4 4 4 4>
      }
      tintColor {colorUseRule=UseColor_Specified, specifiedColor=<rgba 0.0312 0.0312 0.0312 1>}
      uVRegion {bIsValid=false, max=<0 0>, min=<0 0>}
     }
     disabledForeground {colorUseRule=UseColor_ColorTable, specifiedColor=<rgba 0 0 0 0>}
     hovered {
      bIsDynamicallyLoaded=false
      drawAs=RoundedBox
      imageType=NoImage
      mirroring=NoMirror
      resourceName=None
      resourceObject=null
      tiling=NoTile
      imageSize=<32 32>
      margin=<ltrb 0 0 0 0>
      outlineSettings {
       bUseBrushTransparency=false
       roundingType=FixedRadius
       width=0
       color {colorUseRule=UseColor_Foreground, specifiedColor=<rgba 0 0 0 0>}
       cornerRadii=<4 0 0 4>
      }
      tintColor {colorUseRule=UseColor_Specified, specifiedColor=<rgba 0.0953 0.0953 0.0953 1>}
      uVRegion {bIsValid=false, max=<0 0>, min=<0 0>}
     }
     hoveredForeground {colorUseRule=UseColor_ColorTable, specifiedColor=<rgba 0 0 0 0>}
     hoveredSlateSound {resourceObject=null}
     normal {
      bIsDynamicallyLoaded=false
      drawAs=NoDrawType
      imageType=NoImage
      mirroring=NoMirror
      resourceName=None
      resourceObject=null
      tiling=NoTile
      imageSize=<32 32>
      margin=<ltrb 0 0 0 0>
      outlineSettings {
       bUseBrushTransparency=false
       roundingType=FixedRadius
       width=1
       color {colorUseRule=UseColor_ColorTable, specifiedColor=<rgba 0 0 0 0>}
       cornerRadii=<4 4 4 4>
      }
      tintColor {colorUseRule=UseColor_Specified, specifiedColor=<rgba 0.0312 0.0312 0.0312 1>}
      uVRegion {bIsValid=false, max=<0 0>, min=<0 0>}
     }
     normalForeground {colorUseRule=UseColor_ColorTable, specifiedColor=<rgba 0 0 0 0>}
     normalPadding=<ltrb 2 0 2 0>
     pressed {
      bIsDynamicallyLoaded=false
      drawAs=RoundedBox
      imageType=NoImage
      mirroring=NoMirror
      resourceName=None
      resourceObject=null
      tiling=NoTile
      imageSize=<32 32>
      margin=<ltrb 0 0 0 0>
      outlineSettings {
       bUseBrushTransparency=false
       roundingType=FixedRadius
       width=0
       color {colorUseRule=UseColor_Foreground, specifiedColor=<rgba 0 0 0 0>}
       cornerRadii=<4 0 0 4>
      }
      tintColor {colorUseRule=UseColor_Specified, specifiedColor=<rgba 0.0395 0.0395 0.0395 1>}
      uVRegion {bIsValid=false, max=<0 0>, min=<0 0>}
     }
     pressedForeground {colorUseRule=UseColor_ColorTable, specifiedColor=<rgba 0 0 0 0>}
     pressedPadding=<ltrb 2 0.5 2 -0.5>
     pressedSlateSound {resourceObject=null}
    }}
   slot {class=BorderSlot, properties {padding=<ltrb 0 0 0 0>}}
  }
  HorizontalBox_1 {
   widgetClassPath=&HorizontalBox
   parent=Button
   widget=HorizontalBox_1
   slot {class=ButtonSlot, properties {padding=<ltrb 6 5 6 5>}}
  }
  Icon {
   widgetClassPath=/Script/TechAudioToolsContentEditor.TechAudioToolsEditorSlateIcon
   bIsVariable=true
   parent=HorizontalBox_1
   widget=Icon
   slot {class=HorizontalBoxSlot}
  }
  Label_Text {
   widgetClassPath=&TextBlock
   bIsVariable=true
   parent=HorizontalBox_1
   widget=Label_Text
   properties {
    text=Button
    colorAndOpacity {colorUseRule=UseColor_Specified, specifiedColor=<rgba 0.5271 0.5271 0.5271 1>}
    font {
     bForceMonospaced=false
     bMaterialIsStencil=false
     fontMaterial=null
     fontObject=&Roboto
     letterSpacing=0
     monospacedWidth=1
     size=9
     skewAmount=0
     typefaceFontName=Regular
     outlineSettings {
      bApplyOutlineToDropShadows=false
      bMiteredCorners=false
      bSeparateFillAlpha=false
      outlineMaterial=null
      outlineSize=0
      outlineColor=<rgba 0 0 0 1>
     }
    }
   }
   slot {class=HorizontalBoxSlot, properties {verticalAlignment=VAlign_Center, padding=<ltrb 4 1 0 0>}}
  }
  Border {
   widgetClassPath=&Border
   parent=HorizontalBox_0
   widget=Border
   properties {brushColor=<rgba 0.0176 0.0176 0.0176 0>, padding=<ltrb 0 0 0 0>}
   slot {class=HorizontalBoxSlot}
  }
  Menu_Button {
   widgetClassPath=&EditorUtilityButton
   parent=Border
   widget=Menu_Button
   properties {widgetStyle {
     clickedSlateSound {resourceObject=null}
     disabled {
      bIsDynamicallyLoaded=false
      drawAs=NoDrawType
      imageType=NoImage
      mirroring=NoMirror
      resourceName=None
      resourceObject=null
      tiling=NoTile
      imageSize=<32 32>
      margin=<ltrb 0 0 0 0>
      outlineSettings {
       bUseBrushTransparency=false
       roundingType=FixedRadius
       width=1
       color {colorUseRule=UseColor_ColorTable, specifiedColor=<rgba 0 0 0 0>}
       cornerRadii=<4 4 4 4>
      }
      tintColor {colorUseRule=UseColor_Specified, specifiedColor=<rgba 0.0312 0.0312 0.0312 1>}
      uVRegion {bIsValid=false, max=<0 0>, min=<0 0>}
     }
     disabledForeground {colorUseRule=UseColor_ColorTable, specifiedColor=<rgba 0 0 0 0>}
     hovered {
      bIsDynamicallyLoaded=false
      drawAs=RoundedBox
      imageType=NoImage
      mirroring=NoMirror
      resourceName=None
      resourceObject=null
      tiling=NoTile
      imageSize=<32 32>
      margin=<ltrb 0 0 0 0>
      outlineSettings {
       bUseBrushTransparency=false
       roundingType=FixedRadius
       width=0
       color {colorUseRule=UseColor_Foreground, specifiedColor=<rgba 0 0 0 0>}
       cornerRadii=<0 4 4 0>
      }
      tintColor {colorUseRule=UseColor_Specified, specifiedColor=<rgba 0.0953 0.0953 0.0953 1>}
      uVRegion {bIsValid=false, max=<0 0>, min=<0 0>}
     }
     hoveredForeground {colorUseRule=UseColor_ColorTable, specifiedColor=<rgba 0 0 0 0>}
     hoveredSlateSound {resourceObject=null}
     normal {
      bIsDynamicallyLoaded=false
      drawAs=NoDrawType
      imageType=NoImage
      mirroring=NoMirror
      resourceName=None
      resourceObject=null
      tiling=NoTile
      imageSize=<32 32>
      margin=<ltrb 0 0 0 0>
      outlineSettings {
       bUseBrushTransparency=false
       roundingType=FixedRadius
       width=1
       color {colorUseRule=UseColor_ColorTable, specifiedColor=<rgba 0 0 0 0>}
       cornerRadii=<4 4 4 4>
      }
      tintColor {colorUseRule=UseColor_Specified, specifiedColor=<rgba 0.0312 0.0312 0.0312 1>}
      uVRegion {bIsValid=false, max=<0 0>, min=<0 0>}
     }
     normalForeground {colorUseRule=UseColor_ColorTable, specifiedColor=<rgba 0 0 0 0>}
     normalPadding=<ltrb 2 0 2 0>
     pressed {
      bIsDynamicallyLoaded=false
      drawAs=RoundedBox
      imageType=NoImage
      mirroring=NoMirror
      resourceName=None
      resourceObject=null
      tiling=NoTile
      imageSize=<32 32>
      margin=<ltrb 0 0 0 0>
      outlineSettings {
       bUseBrushTransparency=false
       roundingType=FixedRadius
       width=0
       color {colorUseRule=UseColor_Foreground, specifiedColor=<rgba 0 0 0 0>}
       cornerRadii=<0 4 4 0>
      }
      tintColor {colorUseRule=UseColor_Specified, specifiedColor=<rgba 0.0395 0.0395 0.0395 1>}
      uVRegion {bIsValid=false, max=<0 0>, min=<0 0>}
     }
     pressedForeground {colorUseRule=UseColor_ColorTable, specifiedColor=<rgba 0 0 0 0>}
     pressedPadding=<ltrb 2 0.5 2 -0.5>
     pressedSlateSound {resourceObject=null}
    }}
   slot {class=BorderSlot, properties {padding=<ltrb 0 0 0 0>}}
  }
  Menu_Icon {
   widgetClassPath=&TechAudioToolsEditorSlateIcon
   bIsVariable=true
   parent=Menu_Button
   widget=Menu_Icon
   properties {fullIconPath=EditorStyle.Icons.EllipsisVerticalNarrow, imageColorAndOpacity=<rgba 0.5271 0.5271 0.5271 1>, size=<6 20>}
   slot {class=ButtonSlot, properties {padding=<ltrb 3 5 3 5>}}
  }
 ]
}
