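# Editor utility widget for Tech Audio Tools: an icon toggle button that places an editor Slate icon inside a content-slot toggle button.
# PreConstruct applies icon size and initial checked state, then refreshes color, tooltip and brush path.
# Check changes re-broadcast through a dispatcher, update the checkbox, broadcast the "Is Checked" field notification, and refresh those visuals.
# Icon color highlights when checked or hovered, while tooltip and brush path follow checked state.
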EditorUtilityWidgetBlueprint /TechAudioToolsContent/Widgets/Editor/ToggleButtons/EUW_ToggleButton_Icon
graphs [
 EventGraph {nodes [
   n1 = "On Checked State Changed (ToggleButton_Content)"!() -> (OutputDelegate, then, Is Checked:bool) {is=title K2Node_ComponentBoundEvent}
   > Call On Checked State Changed!(execute:Exec, self:Tech Audio Tools Toggle Button Icon Object Reference, Is Checked:bool=@n1.Is Checked) -> (then:Exec) {is=title K2Node_CallDelegate}
   > Set Checked State!(execute, self, Is Checked:bool=@n1.Is Checked) -> (then) {is=title K2Node_CallFunction}
   > Update Icon Color()
   > Update Tool Tip()
   > Update Icon Path()
   Update Icon Color() {is=title K2Node_CustomEvent}
   n2 = IsHovered()
   n3 = Is Checked!(self:Tech Audio Tools Toggle Button Content Slot Object Reference=Get ToggleButton_Content() -> (ToggleButton_Content:Tech Audio Tools Toggle Button Content Slot Object Reference)) -> (Return Value:bool)
   > SetIconColor!(
    execute
    self:Tech Audio Tools Editor Slate Icon Object Reference=Get Icon() -> (Icon:Tech Audio Tools Editor Slate Icon Object Reference)
    InColor:Linear Color Structure=SelectColor(
     A=Get Highlighted Icon Color() -> (Highlighted Icon Color:Linear Color Structure)
     B=Get Default Icon Color() -> (Default Icon Color:Linear Color Structure)
     bPickA=OR Boolean(A=@n3, B=@n2)
    )
   ) -> (then)
   Update Tool Tip() {is=title K2Node_CustomEvent}
   > SetToolTipText!(
    execute
    self:Widget Object Reference
    InToolTipText:"Text (by ref)"=SelectText(
     A=Get Checked Tool Tip() -> (Checked Tool Tip:Text)
     B=Get Unchecked Tool Tip() -> (Unchecked Tool Tip:Text)
     bPickA=Is Checked(self:Self Object Reference)
    )
   ) -> (then)
   Update Icon Path() {is=title K2Node_CustomEvent}
   > SetIconBrushByPath!(
    execute
    self:Tech Audio Tools Editor Slate Icon Object Reference=Get Icon()
    InFullIconPath:Name=Select(
     Option 0:Name=Get Unchecked Icon Brush Path() -> (Unchecked Icon Brush Path:Name=None)
     Option 1:Name=Get Checked Icon Brush Path() -> (Checked Icon Brush Path:Name=None)
     Index=Is Checked()
    ) -> (ReturnValue:Name)
   ) -> (then)
   Event PreConstruct()
   > SetIconSize!(
    execute
    self:Tech Audio Tools Editor Slate Icon Object Reference=Get Icon()
    InSize:Vector 2D Structure=Get Icon Size() -> (Icon Size:Vector 2D Structure)
   ) -> (then)
   > SetCheckedState!(
    execute
    self:Check Box Object Reference=Get ToggleButton!(self:Tech Audio Tools Toggle Button Content Slot Object Reference=Get ToggleButton_Content()) -> (ToggleButton:Check Box Object Reference)
    InCheckedState:ECheckBoxState Enum=Get Checked State() -> (Checked State:ECheckBoxState Enum=Unchecked)
   ) -> (then)
   > Update Icon Color() {is=title K2Node_CallFunction}
   > Update Tool Tip() {is=title K2Node_CallFunction}
   > Update Icon Path() {is=title K2Node_CallFunction}
   Event OnMouseEnter()
   > Update Icon Color()
   Event OnMouseLeave!() -> (OutputDelegate, then, MouseEvent:"Pointer Event Structure (by ref)") {is=title K2Node_Event}
   > Update Icon Color()
  ]}
 Is Checked {nodes [
   Is Checked() {is=title K2Node_FunctionEntry}
   > n1 = Get!(execute:Exec) -> (then:Exec, else:Exec, ToggleButton_Content:Tech Audio Tools Toggle Button Content Slot Object Reference)
   > Return Node!(execute, Return Value:bool=Is Checked(self:Tech Audio Tools Toggle Button Content Slot Object Reference=@n1) {is=title K2Node_CallFunction})
   Return Node(execute=@n1.else, Return Value=false)
  ]}
 On Checked State Changed {nodes [On Checked State Changed!() -> (then, Is Checked:bool=~) {is=title K2Node_FunctionEntry}]}
 Set Checked State {nodes [
   n1 = Set Checked State!() -> (then, Is Checked:bool=~) {is=title K2Node_FunctionEntry}
   > SetCheckedState(
    self=Get ToggleButton(self=Get ToggleButton_Content())
    InCheckedState=Select(Option 0:ECheckBoxState Enum=Unchecked, Option 1:ECheckBoxState Enum=Checked, Index=@n1) -> (ReturnValue:ECheckBoxState Enum)
   )
   > Broadcast Field Value Changed!(execute, self:Widget Object Reference, FieldId:Field Notification Id Structure=(FieldName="Is Checked")) -> (then)
  ]}
]
variables [
 Checked Icon Brush Path {type=string, default=EditorStyle.Icons.Delete}
 Checked State {type=ECheckBoxState, default=Unchecked}
 Checked Tool Tip {type=string, default=""}
 Default Icon Color {type=LinearColor, default=<rgba 0.5271 0.5271 0.5271 1>}
 Highlighted Icon Color {type=LinearColor, default=<rgba 1 1 1 1>}
 Icon Size {type=Vector2D, default=<20 20>}
 On Checked State Changed {default=()}
 Unchecked Icon Brush Path {type=string, default=EditorStyle.Icons.Save}
 Unchecked Tool Tip {type=string, default=""}
]
widgets {
 info {
  namedSlotCount=1
  rootWidgetClass=&/EUW_ToggleButton_ContentSlot.EUW_ToggleButton_ContentSlot_C
  widgetCount=3
 }
 widgets [
  ToggleButton_Content {
   widgetClassPath=&/EUW_ToggleButton_ContentSlot
   bIsVariable=true
   widget=ToggleButton_Content
  }
  VerticalBox_0 {widgetClassPath=&VerticalBox, namedSlotHost=ToggleButton_Content, widget=VerticalBox_0}
  Icon {
   widgetClassPath=/Script/TechAudioToolsContentEditor.TechAudioToolsEditorSlateIcon
   bIsVariable=true
   parent=VerticalBox_0
   widget=Icon
   properties {fullIconPath=EditorStyle.Icons.Save, size=<20 20>}
   slot {class=VerticalBoxSlot, properties {horizontalAlignment=HAlign_Center, verticalAlignment=VAlign_Center, size {sizeRule=Fill, value=1}}}
  }
 ]
}
