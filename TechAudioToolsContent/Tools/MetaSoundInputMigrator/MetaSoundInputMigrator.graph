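# Editor utility widget for migrating MetaSound input default values from a Source MetaSound to linked inputs on a Destination MetaSound or a new preset.
# Two input lists sit side by side; selecting an input highlights its linked partner in the other list.
# Link, unlink and unlink-all buttons call the migrator viewmodel, then clear both selections.
# Source selection can auto-filter destination inputs by data type, and a menu widget handles saved links.
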EditorUtilityWidgetBlueprint /TechAudioToolsContent/Tools/MetaSoundInputMigrator/MetaSoundInputMigrator
graphs [
 Clear List Selections {nodes [
   Clear List Selections() {is=title K2Node_FunctionEntry}
   > Clear Selection(self:List View Object Reference=[
     Get ListView!(self:Tech Audio Tools Editor List View Object Reference=Get SourceMetaSoundInputList() -> (SourceMetaSoundInputList:EUW Meta Sound Input Migrator Input List Object Reference)) -> (ListView:List View Object Reference)
     Get ListView(self=Get DestinationMetaSoundInputList() -> (DestinationMetaSoundInputList:EUW Meta Sound Input Migrator Input List Object Reference))
    ])
  ]}
 Create Save Menu Widget {nodes [
   Create Save Menu Widget() {is=title K2Node_FunctionEntry}
   > n1 = Create Widget!(
    execute:Exec
    WorldContextObject:Object Reference
    Class:User Widget Class Reference=/TechAudioToolsContent/Tools/MetaSoundInputMigrator/Widgets/EUW_MetaSoundInputMigrator_Menu.EUW_MetaSoundInputMigrator_Menu_C
    OwningPlayer:Player Controller Object Reference
    BP_MetaSoundMemberMigratorViewmodel:Meta Sound Member Migrator Viewmodel Object Reference=Get BP_MetaSoundMemberMigratorViewmodel() -> (BP_MetaSoundMemberMigratorViewmodel:Meta Sound Member Migrator Viewmodel Object Reference)
   ) -> (then:Exec, ReturnValue:EUW Meta Sound Input Migrator Menu Object Reference)
   > n2 = Set Save Menu(Save Menu:EUW Meta Sound Input Migrator Menu Object Reference=@n1) -> (Output_Get:EUW Meta Sound Input Migrator Menu Object Reference)
   > Bind Event to On Menu Button Clicked(self:EUW Meta Sound Input Migrator Menu Object Reference=@n2, Delegate=Create Event()) {is=title K2Node_AddDelegate}
   > Return Node!(execute, ReturnValue:User Widget Object Reference=@n2)
  ]}
 EventGraph {nodes [
   "On Primary Button Clicked (SaveLinks_ButtonWithMenu)"() {is=title K2Node_ComponentBoundEvent}
   > Clear List Selections() {is=title K2Node_CallFunction}
   "On Clicked (LinkInputs_Button)"() {is=title K2Node_ComponentBoundEvent}
   > Link Selected Inputs(self:Meta Sound Member Migrator Viewmodel Object Reference=Get BP_MetaSoundMemberMigratorViewmodel())
   > Clear List Selections()
   "On Clicked (UnlinkInputs_Button)"() {is=title K2Node_ComponentBoundEvent}
   > Unlink Selected Inputs(self:Meta Sound Member Migrator Viewmodel Object Reference=Get BP_MetaSoundMemberMigratorViewmodel())
   > Clear List Selections()
   "On Clicked (UnlinkAllInputs_Button)"() {is=title K2Node_ComponentBoundEvent}
   > Unlink All Inputs(self:Meta Sound Member Migrator Viewmodel Object Reference=Get BP_MetaSoundMemberMigratorViewmodel())
   > Clear List Selections()
   Handle Menu Button Clicked() {is=title K2Node_CustomEvent}
   > Close(self:Menu Anchor Object Reference=Get MenuAnchor() -> (MenuAnchor:Menu Anchor Object Reference))
   > Clear List Selections()
  ]}
 Get Link Inputs Button Visibility {nodes [
   n1 = Get Link Inputs Button Visibility!() -> (then, Can Unlink Selected Inputs:bool=~) {is=title K2Node_FunctionEntry}
   > Return Node!(execute, Visibility:ESlateVisibility Enum=Select(Option 0=Visible, Option 1=Collapsed, Index=@n1))
  ]}
 Get Unlink Inputs Button Visibility {nodes [
   n1 = Get Unlink Inputs Button Visibility!() -> (then, Can Unlink Selected Inputs:bool=~) {is=title K2Node_FunctionEntry}
   > Return Node(Visibility=Select(Option 0=Collapsed, Option 1=Visible, Index=@n1))
  ]}
 Handle Automatically Filter Data Types Changed {nodes [
   Handle Automatically Filter Data Types Changed!() -> (then, Automatically Filter Data Types:bool=~) {is=title K2Node_FunctionEntry}
   > n1 = Sequence()
   > n2 = Get!(execute:Exec) -> (then:Exec, else:Exec, BP_MetaSoundMemberMigratorViewmodel:Meta Sound Member Migrator Viewmodel Object Reference)
   > n3 = Get!(execute, self:Meta Sound Member Migrator Viewmodel Object Reference=@n2) -> (then, else, Source Migrator List Viewmodel:BP Meta Sound Input Migrator List View Model Object Reference)
   n4 = GetSelectedInputEditorViewModel!(self:Meta Sound Input List Editor Viewmodel Object Reference=@n3) -> (ReturnValue:Meta Sound Input Editor Viewmodel Object Reference)
   > Is Valid(InputObject=@n4)
   > Add(
    TargetArray:Array of Names=Get Included Data Types() -> (Included Data Types:Array of Names)
    NewItem:"Name (by ref)"=Get DataType!(self:Meta Sound Input Viewmodel Object Reference=@n4) -> (DataType:Name=None)
   ) {is=title K2Node_CallArrayFunction}
   Sequence(execute=@n1.then_1)
   n5 = Get MetaSoundInputMigratorListViewModel!(self:EUW Meta Sound Input Migrator Input List Object Reference=Get DestinationMetaSoundInputList()) -> (MetaSoundInputMigratorListViewModel:BP Meta Sound Input Migrator List View Model Object Reference)
   > Set IncludedDataTypes!(
    execute
    IncludedDataTypes:Array of Names=Select(
     Option 0:Array of Names
     Option 1:Array of Names=Get Included Data Types()
     Index=Get Automatically Filter Data Types() -> (Automatically Filter Data Types:bool)
    ) -> (ReturnValue:Array of Names)
    self:Meta Sound Member List Viewmodel Object Reference=@n5
   ) -> (then, Output_Get:Array of Names)
   > Broadcast Field Value Changed!(execute, Object:Object Reference=@n5, FieldId:Field Notification Id Structure=(FieldName="InputEditorEntryViewModels")) -> (then)
  ]}
 Handle Selected Destination Input Changed {nodes [
   Handle Selected Destination Input Changed!() -> (then, Entry Viewmodel:Tech Audio Tools Entry Viewmodel Object Reference) {is=title K2Node_FunctionEntry}
   > n1 = Sequence()
   > n2 = Get!(execute) -> (then, else, Entry Viewmodel:Tech Audio Tools Entry Viewmodel Object Reference)
   > n3 = Cast To BP_MetaSoundInputMigratorEntryViewModel!(execute, Object=@n2) -> (then, CastFailed, AsBP Meta Sound Input Migrator Entry View Model:BP Meta Sound Input Migrator Entry View Model Object Reference) {is=title K2Node_DynamicCast}
   > Set Migrator Entry Viewmodel(Migrator Entry Viewmodel:BP Meta Sound Input Migrator Entry View Model Object Reference=@n3) -> (Output_Get:BP Meta Sound Input Migrator Entry View Model Object Reference)
   n4 = Sequence(execute=@n1.then_1)
   > n5 = Get!(execute) -> (then, else, Migrator Entry Viewmodel:BP Meta Sound Input Migrator Entry View Model Object Reference)
   > n6 = Get!(execute, self:BP Meta Sound Input Migrator Entry View Model Object Reference=@n5) -> (then, else, Linked Migrator Entry Viewmodel:BP Meta Sound Input Migrator Entry View Model Object Reference)
   > Set Selected Item!(execute, self:List View Object Reference=Get ListView(self=Get SourceMetaSoundInputList()), Item:Object Reference=@n6) -> (then) {is=title K2Node_CallFunction}
   n7 = Get!(execute=@n5.else) -> (then, else, Previously Selected Destination Input:Tech Audio Tools Entry Viewmodel Object Reference)
   > n8 = Cast To BP_MetaSoundInputMigratorEntryViewModel(Object=@n7)
   > Get!(execute, self:BP Meta Sound Input Migrator Entry View Model Object Reference=@n8) -> (then, else, Linked Migrator Entry Viewmodel:BP Meta Sound Input Migrator Entry View Model Object Reference)
   > Clear Selection(self=Get ListView(self=Get SourceMetaSoundInputList()))
   Sequence(execute=@n4.then_1)
   > Set Previously Selected Destination Input(Previously Selected Destination Input:Tech Audio Tools Entry Viewmodel Object Reference=Get Entry Viewmodel() -> (Entry Viewmodel:Tech Audio Tools Entry Viewmodel Object Reference)) -> (Output_Get:Tech Audio Tools Entry Viewmodel Object Reference)
  ]}
 Handle Selected Source Input Changed {nodes [
   Handle Selected Source Input Changed!() -> (then, Entry Viewmodel:Tech Audio Tools Entry Viewmodel Object Reference) {is=title K2Node_FunctionEntry}
   > n1 = Sequence()
   > n2 = Get!(execute) -> (then, else, Entry Viewmodel:Tech Audio Tools Entry Viewmodel Object Reference)
   > n3 = Cast To BP_MetaSoundInputMigratorEntryViewModel(Object=@n2)
   > Set Migrator Entry Viewmodel(Migrator Entry Viewmodel=@n3)
   n4 = Sequence(execute=@n1.then_1)
   > n5 = Get!(execute) -> (then, else, Migrator Entry Viewmodel:BP Meta Sound Input Migrator Entry View Model Object Reference)
   > n6 = Get!(execute, self:BP Meta Sound Input Migrator Entry View Model Object Reference=@n5) -> (then, else, Linked Migrator Entry Viewmodel:BP Meta Sound Input Migrator Entry View Model Object Reference)
   > Set Selected Item(self=Get ListView(self=Get DestinationMetaSoundInputList()), Item=@n6)
   n7 = Get!(execute=@n5.else) -> (then, else, Previously Selected Source Entry:Tech Audio Tools Entry Viewmodel Object Reference)
   > n8 = Cast To BP_MetaSoundInputMigratorEntryViewModel(Object=@n7)
   > Get!(execute, self:BP Meta Sound Input Migrator Entry View Model Object Reference=@n8) -> (then, else, Linked Migrator Entry Viewmodel:BP Meta Sound Input Migrator Entry View Model Object Reference)
   > Clear Selection(self=Get ListView(self=Get DestinationMetaSoundInputList()))
   n9 = Sequence(execute=@n4.then_1)
   > Handle Automatically Filter Data Types Changed!(
    execute
    self
    Automatically Filter Data Types:bool=Get Automatically Filter Data Types!(self:Meta Sound Member Migrator Viewmodel Object Reference=Get BP_MetaSoundMemberMigratorViewmodel()) -> (Automatically Filter Data Types)
   ) -> (then) {is=title K2Node_CallFunction}
   Sequence(execute=@n9.then_1)
   > Set Previously Selected Source Entry(Previously Selected Source Entry:Tech Audio Tools Entry Viewmodel Object Reference=Get Entry Viewmodel()) -> (Output_Get:Tech Audio Tools Entry Viewmodel Object Reference)
  ]}
]
variables [
 Previously Selected Destination Input {type=TechAudioToolsEntryViewModel, default=null}
 Previously Selected Source Entry {type=TechAudioToolsEntryViewModel, default=null}
 Save Menu {type=EUW_MetaSoundInputMigrator_Menu_C, default=null}
]
widgets {
 info {namedSlotCount=2, rootWidgetClass=&Border, widgetCount=32}
 widgets [
  Border_0 {widgetClassPath=&Border, widget=Border_0, properties {brushColor=<rgba 0.0075 0.0075 0.0075 1>, padding=<ltrb 0 0 0 0>}}
  VerticalBox_32 {
   widgetClassPath=&VerticalBox
   parent=Border_0
   widget=VerticalBox_32
   slot {class=BorderSlot, properties {padding=<ltrb 0 0 0 0>}}
  }
  Border_12 {
   widgetClassPath=&Border
   parent=VerticalBox_32
   widget=Border_12
   properties {brushColor=<rgba 0.0176 0.0176 0.0176 1>, padding=<ltrb 0 0 0 0>}
   slot {class=VerticalBoxSlot}
  }
  HorizontalBox_5 {
   widgetClassPath=&HorizontalBox
   parent=Border_12
   widget=HorizontalBox_5
   slot {class=BorderSlot, properties {padding=<ltrb 0 0 0 0>}}
  }
  VerticalBox_2 {
   widgetClassPath=&VerticalBox
   parent=HorizontalBox_5
   widget=VerticalBox_2
   slot {class=HorizontalBoxSlot, properties {padding=<ltrb 6 6 6 6>}}
  }
  SaveLinks_ButtonWithMenu {
   widgetClassPath=/TechAudioToolsContent/Widgets/Editor/Buttons/EUW_Button_WithMenu
   bIsVariable=true
   parent=VerticalBox_2
   widget=SaveLinks_ButtonWithMenu
   properties {
    label Text=Save Links
    toolTipText=Saves all currently-linked inputs as AssetUserData on the Destination MetaSound. Saved links are automatically recalled when data matching the currently-selected assets is found.
   }
   slot {class=VerticalBoxSlot, properties {verticalAlignment=VAlign_Center, size {sizeRule=Fill, value=1}}}
  }
  MenuAnchor {
   widgetClassPath=/Script/UMG.MenuAnchor
   bIsVariable=true
   parent=VerticalBox_2
   widget=MenuAnchor
   properties {placement=MenuPlacement_BelowAnchor}
   slot {class=VerticalBoxSlot}
  }
  Border {
   widgetClassPath=&Border
   parent=HorizontalBox_5
   widget=Border
   properties {brushColor=<rgba 0.0103 0.0103 0.0103 1>, padding=<ltrb 0 0 0 0>}
   slot {class=HorizontalBoxSlot, properties {padding=<ltrb 0 1 0 1>}}
  }
  Spacer {
   widgetClassPath=&Spacer
   parent=Border
   widget=Spacer
   properties {size=<2 1>}
   slot {class=BorderSlot, properties {padding=<ltrb 0 0 0 0>}}
  }
  WrapBox {
   widgetClassPath=&WrapBox
   parent=HorizontalBox_5
   widget=WrapBox
   properties {bOverride_Cursor=true, horizontalAlignment=HAlign_Fill, innerSlotPadding=<6 6>}
   slot {
    class=HorizontalBoxSlot
    properties {horizontalAlignment=HAlign_Right, verticalAlignment=VAlign_Center, padding=<ltrb 6 6 6 6>, size {sizeRule=Fill, value=1}}
   }
  }
  ShowValues_ToggleButton {
   widgetClassPath=/TechAudioToolsContent/Widgets/Editor/ToggleButtons/EUW_ToggleButton_Icon
   bIsVariable=true
   parent=WrapBox
   widget=ShowValues_ToggleButton
   properties {
    checked Icon Brush Path=EditorStyle.EditorViewport.Visualizers
    checked Tool Tip=Hide default values.
    unchecked Icon Brush Path=EditorStyle.EditorViewport.Visualizers
    unchecked Tool Tip=Show default values.
   }
   slot {class=WrapBoxSlot}
  }
  AutoFilter_ToggleButton {
   widgetClassPath=&EUW_ToggleButton_Icon
   bIsVariable=true
   parent=WrapBox
   widget=AutoFilter_ToggleButton
   properties {
    checked Tool Tip=Disable automatic data type filtering.
    unchecked Icon Brush Path=CoreStyle.Icons.AutoFilter
    unchecked Tool Tip=Enable automatic data type filtering.
    visibility=Collapsed
   }
   slot {class=WrapBoxSlot}
  }
  Border_2 {
   widgetClassPath=&Border
   parent=VerticalBox_32
   widget=Border_2
   properties {
    background {
     bIsDynamicallyLoaded=false
     drawAs=RoundedBox
     imageType=NoImage
     mirroring=NoMirror
     resourceName=None
     resourceObject=null
     tiling=NoTile
     imageSize=<32 32>
     margin=<ltrb 0 0 0 0>
     outlineSettings {
      bUseBrushTransparency=false
      roundingType=FixedRadius
      width=0
      color {colorUseRule=UseColor_Specified, specifiedColor=<rgba 0 0 0 0>}
      cornerRadii=<4 4 4 4>
     }
     tintColor {colorUseRule=UseColor_Specified, specifiedColor=<rgba 1 1 1 1>}
     uVRegion {bIsValid=false, max=<0 0>, min=<0 0>}
    }
    brushColor=<rgba 0.0103 0.0103 0.0103 1>
    padding=<ltrb 0 0 0 0>
   }
   slot {class=VerticalBoxSlot, properties {padding=<ltrb 4 4 4 4>, size {sizeRule=Fill, value=1}}}
  }
  HorizontalBox {
   widgetClassPath=&HorizontalBox
   parent=Border_2
   widget=HorizontalBox
   slot {class=BorderSlot, properties {padding=<ltrb 0 0 0 0>}}
  }
  SourceMetaSoundInputList {
   widgetClassPath=/TechAudioToolsContent/Tools/MetaSoundInputMigrator/Widgets/EUW_MetaSoundInputMigrator_InputList
   bIsVariable=true
   parent=HorizontalBox
   widget=SourceMetaSoundInputList
   properties {
    header Text=Source MetaSound
    instructions Text=Select a MetaSound to migrate input defaults from.
    metaSoundInputMigratorListViewModel=&self:WidgetTree.SourceMetaSoundInputList.MetaSoundInputListEditorViewModel_0
    techAudioToolsListViewModel=&self:WidgetTree.SourceMetaSoundInputList.TechAudioToolsListViewModel_0
   }
   slot {class=HorizontalBoxSlot, properties {size {sizeRule=Fill, value=1}}}
  }
  Border_3 {
   widgetClassPath=&Border
   parent=HorizontalBox
   widget=Border_3
   properties {brushColor=<rgba 0.0075 0.0075 0.0075 1>, padding=<ltrb 0 0 0 0>}
   slot {class=HorizontalBoxSlot}
  }
  Spacer_64 {
   widgetClassPath=&Spacer
   parent=Border_3
   widget=Spacer_64
   properties {size=<4 2>}
   slot {class=BorderSlot, properties {padding=<ltrb 0 0 0 0>}}
  }
  DestinationMetaSoundInputList {
   widgetClassPath=&EUW_MetaSoundInputMigrator_InputList
   bIsVariable=true
   parent=HorizontalBox
   widget=DestinationMetaSoundInputList
   properties {
    header Icon Path=EditorStyle.Icons.CircleArrowDown
    header Text=Destination MetaSound
    instructions Icon Path=CoreStyle.Icons.CircleArrowDown
    instructions Text=Select a MetaSound to migrate input defaults to.
    metaSoundInputMigratorListViewModel=&self:WidgetTree.DestinationMetaSoundInputList.MetaSoundInputListEditorViewModel_0
    techAudioToolsListViewModel=&self:WidgetTree.DestinationMetaSoundInputList.TechAudioToolsListViewModel_0
   }
   slot {class=HorizontalBoxSlot, properties {size {sizeRule=Fill, value=1}}}
  }
  Border_1 {
   widgetClassPath=&Border
   parent=VerticalBox_32
   widget=Border_1
   properties {
    background {
     bIsDynamicallyLoaded=false
     drawAs=Image
     imageType=NoImage
     mirroring=NoMirror
     resourceName=None
     resourceObject=null
     tiling=NoTile
     imageSize=<32 32>
     margin=<ltrb 0 0 0 0>
     outlineSettings {
      bUseBrushTransparency=false
      roundingType=HalfHeightRadius
      width=0
      color {colorUseRule=UseColor_Specified, specifiedColor=<rgba 0 0 0 0>}
      cornerRadii=<0 0 0 0>
     }
     tintColor {colorUseRule=UseColor_Specified, specifiedColor=<rgba 0.0176 0.0176 0.0176 1>}
     uVRegion {bIsValid=false, max=<0 0>, min=<0 0>}
    }
    padding=<ltrb 6 6 6 6>
   }
   slot {class=VerticalBoxSlot}
  }
  HorizontalBox_1 {
   widgetClassPath=&HorizontalBox
   parent=Border_1
   widget=HorizontalBox_1
   slot {class=BorderSlot, properties {padding=<ltrb 6 6 6 6>}}
  }
  LinkInputs_Button {
   widgetClassPath=&EditorUtilityButton
   parent=HorizontalBox_1
   widget=LinkInputs_Button
   properties {toolTipText=Links the selected Source MetaSound input with the selected Destination MetaSound input.}
   slot {class=HorizontalBoxSlot, properties {horizontalAlignment=HAlign_Right}}
  }
  CommonTextBlock_1 {
   widgetClassPath=&CommonTextBlock
   parent=LinkInputs_Button
   widget=CommonTextBlock_1
   properties {
    text=Link Selected
    colorAndOpacity {colorUseRule=UseColor_Specified, specifiedColor=<rgba 0.5271 0.5271 0.5271 1>}
    font {
     bForceMonospaced=false
     bMaterialIsStencil=false
     fontMaterial=null
     fontObject=&Roboto
     letterSpacing=0
     monospacedWidth=1
     size=10
     skewAmount=0
     typefaceFontName=Bold
     outlineSettings {
      bApplyOutlineToDropShadows=false
      bMiteredCorners=false
      bSeparateFillAlpha=false
      outlineMaterial=null
      outlineSize=0
      outlineColor=<rgba 0 0 0 1>
     }
    }
   }
   slot {class=ButtonSlot}
  }
  UnlinkInputs_Button {
   widgetClassPath=&EditorUtilityButton
   parent=HorizontalBox_1
   widget=UnlinkInputs_Button
   properties {visibility=Hidden}
   slot {class=HorizontalBoxSlot, properties {horizontalAlignment=HAlign_Right}}
  }
  CommonTextBlock {
   widgetClassPath=&CommonTextBlock
   parent=UnlinkInputs_Button
   widget=CommonTextBlock
   properties {
    text=Unlink Selected
    colorAndOpacity {colorUseRule=UseColor_Specified, specifiedColor=<rgba 0.5271 0.5271 0.5271 1>}
    font {
     bForceMonospaced=false
     bMaterialIsStencil=false
     fontMaterial=null
     fontObject=&Roboto
     letterSpacing=0
     monospacedWidth=1
     size=10
     skewAmount=0
     typefaceFontName=Bold
     outlineSettings {
      bApplyOutlineToDropShadows=false
      bMiteredCorners=false
      bSeparateFillAlpha=false
      outlineMaterial=null
      outlineSize=0
      outlineColor=<rgba 0 0 0 1>
     }
    }
   }
   slot {class=ButtonSlot}
  }
  UnlinkAllInputs_Button {
   widgetClassPath=&EditorUtilityButton
   parent=HorizontalBox_1
   widget=UnlinkAllInputs_Button
   properties {toolTipText=Unlinks all currently-linked inputs.}
   slot {class=HorizontalBoxSlot, properties {horizontalAlignment=HAlign_Right, padding=<ltrb 4 0 0 0>}}
  }
  CommonTextBlock_4 {
   widgetClassPath=&CommonTextBlock
   parent=UnlinkAllInputs_Button
   widget=CommonTextBlock_4
   properties {
    text=Unlink All Inputs
    colorAndOpacity {colorUseRule=UseColor_Specified, specifiedColor=<rgba 0.5271 0.5271 0.5271 1>}
    font {
     bForceMonospaced=false
     bMaterialIsStencil=false
     fontMaterial=null
     fontObject=&Roboto
     letterSpacing=0
     monospacedWidth=1
     size=10
     skewAmount=0
     typefaceFontName=Bold
     outlineSettings {
      bApplyOutlineToDropShadows=false
      bMiteredCorners=false
      bSeparateFillAlpha=false
      outlineMaterial=null
      outlineSize=0
      outlineColor=<rgba 0 0 0 1>
     }
    }
   }
   slot {class=ButtonSlot}
  }
  OverwriteInputs_Button {
   widgetClassPath=/TechAudioToolsContent/Widgets/Editor/Buttons/EUW_Button_Secondary
   parent=HorizontalBox_1
   widget=OverwriteInputs_Button
   properties {toolTipText=Copies the input default values from the Source MetaSound inputs to the linked inputs on the Destination MetaSound.}
   slot {class=HorizontalBoxSlot, properties {horizontalAlignment=HAlign_Right, padding=<ltrb 4 0 0 0>, size {sizeRule=Fill, value=1}}}
  }
  VerticalBox {widgetClassPath=&VerticalBox, namedSlotHost=OverwriteInputs_Button, widget=VerticalBox}
  CommonTextBlock_2 {
   widgetClassPath=&CommonTextBlock
   parent=VerticalBox
   widget=CommonTextBlock_2
   properties {
    text=Overwrite Destination Inputs
    font {
     bForceMonospaced=false
     bMaterialIsStencil=false
     fontMaterial=null
     fontObject=&Roboto
     letterSpacing=0
     monospacedWidth=1
     size=10
     skewAmount=0
     typefaceFontName=Bold
     outlineSettings {
      bApplyOutlineToDropShadows=false
      bMiteredCorners=false
      bSeparateFillAlpha=false
      outlineMaterial=null
      outlineSize=0
      outlineColor=<rgba 0 0 0 1>
     }
    }
   }
   slot {class=VerticalBoxSlot, properties {horizontalAlignment=HAlign_Center, verticalAlignment=VAlign_Center, padding=<ltrb 12 2 12 2>}}
  }
  CreateNewPreset_Button {
   widgetClassPath=/TechAudioToolsContent/Widgets/Editor/Buttons/EUW_Button_Primary
   parent=HorizontalBox_1
   widget=CreateNewPreset_Button
   properties {toolTipText=Creates a new MetaSound preset based on the Destination MetaSound and migrates linked input values from the Source MetaSound.}
   slot {class=HorizontalBoxSlot, properties {padding=<ltrb 4 0 0 0>}}
  }
  VerticalBox_0 {widgetClassPath=&VerticalBox, namedSlotHost=CreateNewPreset_Button, widget=VerticalBox_0}
  CommonTextBlock_3 {
   widgetClassPath=&CommonTextBlock
   parent=VerticalBox_0
   widget=CommonTextBlock_3
   properties {
    text=Create New Preset
    font {
     bForceMonospaced=false
     bMaterialIsStencil=false
     fontMaterial=null
     fontObject=&Roboto
     letterSpacing=0
     monospacedWidth=1
     size=10
     skewAmount=0
     typefaceFontName=Bold
     outlineSettings {
      bApplyOutlineToDropShadows=false
      bMiteredCorners=false
      bSeparateFillAlpha=false
      outlineMaterial=null
      outlineSize=0
      outlineColor=<rgba 0 0 0 1>
     }
    }
   }
   slot {class=VerticalBoxSlot, properties {horizontalAlignment=HAlign_Center, verticalAlignment=VAlign_Center, padding=<ltrb 12 2 12 2>}}
  }
 ]
}
